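# Physics asset for the ElegantLilyAnkle backpack cosmetic, giving its five petal bones collision bodies and joint limits for physics-driven motion.
# Each petal has a same-sized capsule constrained directly to a small capsule on the pack bone, forming a star of independent joints rather than a chain.
# Tight swing limits and looser twist limits restrict petal motion, with the fifth petal getting a narrower twist and a further-offset capsule.
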PhysicsAsset /BRCosmetics/Accessories/FORT_Backpacks/Backpack_F_MED_ElegantLilyAnkle/Meshes/F_MED_ElegantLilyAnkle_Pack_Physics
bodies {
 dyn_petal_01 [{shapeType=Capsule, length=5, radius=4, shapeName=dyn_petal_01_capsule, center=<0 0 1.7064>}]
 dyn_petal_02 [{shapeType=Capsule, length=5, radius=4, shapeName=dyn_petal_02_capsule, center=<0 0 1.7064>}]
 dyn_petal_03 [{shapeType=Capsule, length=5, radius=4, shapeName=dyn_petal_03_capsule, center=<0 0 1.7064>}]
 dyn_petal_04 [{shapeType=Capsule, length=5, radius=4, shapeName=dyn_petal_04_capsule, center=<0 0 1.7064>}]
 dyn_petal_05 [{shapeType=Capsule, length=5, radius=4, shapeName=dyn_petal_05_capsule, center=<0 0 3.6046>}]
 pack [{shapeType=Capsule, length=1, radius=1}]
}
constraints [
 {bone1Name=dyn_petal_01, bone2Name=pack, swing1LimitDegrees=5, swing2LimitDegrees=7, twistLimitDegrees=45}
 {bone1Name=dyn_petal_02, bone2Name=pack, swing1LimitDegrees=5, swing2LimitDegrees=7, twistLimitDegrees=45}
 {bone1Name=dyn_petal_03, bone2Name=pack, swing1LimitDegrees=5, swing2LimitDegrees=7, twistLimitDegrees=45}
 {bone1Name=dyn_petal_04, bone2Name=pack, swing1LimitDegrees=5, swing2LimitDegrees=7, twistLimitDegrees=45}
 {bone1Name=dyn_petal_05, bone2Name=pack, swing1LimitDegrees=5, swing2LimitDegrees=7, twistLimitDegrees=25}
]
